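# DecalFade is a Default Lit material whose math forms a timed jump-and-shrink position offset, driven by a SpawnTime parameter.
# A half-sine over AirTime lifts along Z by Jump Height and returns.
# After AirTime, a FadeTime ramp adds an offset pulling each point into the actor position, collapsing the shape.
# A separate branch scales points around the object position by Scale; no material output connections appear.
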Material /Volumetrics/Tools/FluidSimulation/Blueprints/Pawn/DecalFade
graph {
 properties {pixelDepthOffsetMode=PDOM_Legacy, translucencyPass=MTP_BeforeDOF, usedShadingModels=Default Lit}
 nodes [
  n1 = Subtract(A=Time(), B=ScalarParameter() {parameterName=SpawnTime})
  n2 = ScalarParameter() {defaultValue=1, parameterName=AirTime}
  n3 = Sine(None=Saturate(None=Divide(A=@n1, B=@n2))) {desc="Period = 2", period=2}
  n4 = Constant3Vector() {constant=<rgba 0 0 1 0>}
  n5 = Subtract(A=Time(), B=ScalarParameter() {parameterName=SpawnTime})
  n6 = ScalarParameter() {defaultValue=1, parameterName=AirTime}
  n7 = Divide(A=Max(A=Subtract(A=@n5, B=@n6)) {constB=0}, B=ScalarParameter() {defaultValue=1, parameterName=FadeTime})
  n8 = ActorPositionWS()
  n9 = WorldPosition()
  Add(
   A=Multiply(A=Multiply(A=@n3, B=@n4), B=ScalarParameter() {defaultValue=256, parameterName=Jump Height})
   B=Multiply(A=Saturate(None=@n7), B=Subtract(A=@n8, B=@n9.XYZ))
  )
  Constant() {r=0.2}
  Multiply(
   A=OneMinus(None=ScalarParameter() {parameterName=Scale, sliderMax=5})
   B=Subtract(A=ObjectPositionWS() {desc=Scale Point}, B=WorldPosition().XYZ)
  )
 ]
}
