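; FakeDecal is a Blueprint in the fluid simulation tools that, by its name, stands in for a decal using a static mesh.
; On BeginPlay it creates a dynamic material instance on slot 0 of that mesh and writes the current game time to SpawnTime and its LifeTime variable to LifeTime.
; Any change over that lifetime is left to the material, as the graph has no tick or destroy logic.
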(event EventBeginPlay
  (bind _returnvalue (Rendering|Material|CreateDynamicMaterialInstance (Variables|Default|GetStaticMesh) 0 0 "None" :!class /Script/Engine.PrimitiveComponent))
  (Rendering|Material|SetScalarParameterValue _returnvalue "SpawnTime" (Utilities|Time|GetTimeSeconds) :!class /Script/Engine.MaterialInstanceDynamic)
  (Rendering|Material|SetScalarParameterValue _returnvalue "LifeTime" (Variables|Default|GetLifeTime) :!class /Script/Engine.MaterialInstanceDynamic))

(fn ConstructionScript)
