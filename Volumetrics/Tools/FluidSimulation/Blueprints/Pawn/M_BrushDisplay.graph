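# Translucent, unlit, two-sided material that, going by its name and folder, displays the brush of the fluid simulation tool's pawn.
# An inverted depth fade (FD1) multiplied by a second, shorter one (FD2) forms a soft band where the mesh nears scene geometry.
# An Opacity parameter scales that band, and a vector parameter E supplies a color, cyan by default.
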Material /Volumetrics/Tools/FluidSimulation/Blueprints/Pawn/M_BrushDisplay
graph {
 properties {blendMode=BLEND_Translucent, pixelDepthOffsetMode=PDOM_Legacy, shadingModel=MSM_Unlit, twoSided=true, usedShadingModels=Unlit}
 nodes [
  n1 = DepthFade(FadeDistance=ScalarParameter() {defaultValue=32, parameterName=FD1})
  n2 = ScalarParameter() {defaultValue=8, parameterName=FD2}
  Multiply(A=Multiply(A=OneMinus(None=@n1), B=DepthFade(FadeDistance=@n2)), B=ScalarParameter() {defaultValue=1, parameterName=Opacity})
  VectorParameter() {parameterName=E, defaultValue=<rgba 0 1 1 0>}
 ]
}
