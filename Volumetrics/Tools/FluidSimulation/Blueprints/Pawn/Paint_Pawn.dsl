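; Player pawn that lets the mouse cursor drive a BP_FluidSimulation_01 actor, found at BeginPlay when the cursor is shown.
; Each tick traces under the cursor, derives mouse velocity from the change in hit position and moves a brush preview there.
; While Pressed, it writes obstacle location, brush-scaled radius and force velocity to the simulation; otherwise Right sets force location.
; Mouse button, wheel and A/Z events contain no effective logic.
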(event EventBeginPlay
  ;[1 1 1] On Begin Play
  (Variables|/Script/Engine.PlayerController|SetbShowMouseCursor true (Game|GetPlayerController 0))
  (bind _outactors (Actor|GetAllActorsOfClass "/Volumetrics/Tools/FluidSimulation/Blueprints/BP_FluidSimulation_01.BP_FluidSimulation_01_C"))
  (Variables|Default|SetFluidSimBP (Utilities|Array|Get(acopy) _outactors))
  (bind _outactors_2 (Actor|GetAllActorsOfClass 0)))

(event LeftMouseButton
  (params (Key :type Key%20Structure))
  (bind _fluidsimbp (Variables|Default|GetFluidSimBP)))

(event RightMouseButton
  (params (Key :type Key%20Structure))
  (bind _fluidsimbp (Variables|Default|GetFluidSimBP)))

(event EventTick
  (params (DeltaSeconds :type Float%20%28single-precision%29))
  (bind _fluidsimbp (Variables|Default|GetFluidSimBP))
  (CallFunction|LineTraceUnderCursor)
  (CallFunction|UpdateBrushPreview)
  (if (Variables|Default|GetPressed)
    (Variables|Default|SetVelocity (* (Variables|Default|GetMouseVelocity) (Variables|Default|GetOpacity)))
    (if (not (Utilities|Equal(Object) (Variables|Default|GetHit%20Actor) _fluidsimbp))
      (Utilities|IsValid _fluidsimbp
        (:Is%20Valid
          _)))
    (Variables|/Volumetrics/Tools/FluidSimulation/Blueprints/BP_FluidSimulation_01.BP_FluidSimulation_01_C|SetObstacle%20Location (Variables|Default|GetHit%20Position) (Variables|Default|GetFluidSimBP))
    (Variables|/Volumetrics/Tools/FluidSimulation/Blueprints/BP_FluidSimulation_01.BP_FluidSimulation_01_C|SetForce%20Velocity (Variables|Default|GetMouseVelocity) (Variables|Default|GetFluidSimBP))
    (Variables|/Volumetrics/Tools/FluidSimulation/Blueprints/BP_FluidSimulation_01.BP_FluidSimulation_01_C|SetObstacle%20Radius (* (Variables|Default|GetBrushSize) 0.5) (Variables|Default|GetFluidSimBP))
    (elif (Variables|Default|GetRight)
      (Variables|/Volumetrics/Tools/FluidSimulation/Blueprints/BP_FluidSimulation_01.BP_FluidSimulation_01_C|SetForce%20Location (Variables|Default|GetHit%20Position) (Variables|Default|GetFluidSimBP)))))

(event MouseWheelUp
  (params (Key :type Key%20Structure))
  (bind _fluidsimbp (Variables|Default|GetFluidSimBP)))

(event MouseWheelDown
  (params (Key :type Key%20Structure))
  (bind _fluidsimbp (Variables|Default|GetFluidSimBP)))

(event A
  (params (Key :type Key%20Structure)))

(event Z
  (params (Key :type Key%20Structure)))

(fn LineTraceUnderCursor
  (bind _returnvalue (Game|GetPlayerController 0))
  (bind (_hitresult _returnvalue_2) (Game|Player|GetHitResultUnderCursorbyChannel _returnvalue "TraceTypeQuery1" true))
  (bind (_bblockinghit _binitialoverlap _time _distance _location _impactpoint _normal _impactnormal _physmat _hitactor) (Collision|BreakHitResult _hitresult))
  (if _returnvalue_2
    (Variables|Default|SetMouseVelocity (- _location (Variables|Default|GetHit%20Position)))
    (Variables|Default|SetHit%20Position _location)
    (Variables|Default|SetHit%20Actor _hitactor)))

(fn UpdateBrushPreview
  (Transformation|SetWorldLocation (Variables|Default|GetBrushPreview) (Variables|Default|GetHit%20Position) false false))

(fn ConstructionScript)
