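# Translucent unlit debug material in the FluidSimulation pawn tools that visualises the current Time value through DebugScalarValues.
# It also computes a lifetime fade: Time minus SpawnTime plus LifeTime, divided by FadeTime, saturated and inverted, so the value drops from one to zero after the lifetime ends.
# An AnalogHat texture's alpha feeds a multiply, but the graph shows no links from these chains to material outputs.
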Material /Volumetrics/Tools/FluidSimulation/Blueprints/Pawn/timedebug
graph {
 properties {blendMode=BLEND_Translucent, pixelDepthOffsetMode=PDOM_Legacy, shadingModel=MSM_Unlit, usedShadingModels=Unlit}
 nodes [
  n1 = Time()
  MaterialFunctionCall(Number=@n1) {materialFunction=&DebugScalarValues}
  Multiply(A=TextureSample() {texture=/Engine/MobileResources/HUD/AnalogHat}.A)
  n2 = Subtract(A=@n1, B=Add(A=ScalarParameter() {parameterName=SpawnTime}, B=ScalarParameter() {defaultValue=2, parameterName=LifeTime}))
  n3 = ScalarParameter() {defaultValue=1, parameterName=FadeTime}
  OneMinus(None=Saturate(None=Divide(A=@n2, B=@n3)))
 ]
}
